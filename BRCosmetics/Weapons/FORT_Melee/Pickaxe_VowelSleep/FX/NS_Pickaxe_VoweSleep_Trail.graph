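# Niagara ribbon trail effect for the VowelSleep pickaxe, apparently shown during melee swings.
# A single emitter spawns two anim trails from bone or socket indices, with spawn group selecting each trail's color and short lifetime.
# System-level FadeIn scales spawn alpha, while particle update multiplies color by a time-of-day tint and alpha by FadeOut.
# User parameters expose front-end flags, a ribbon ID offset and a swing section index.
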NiagaraSystem /BRCosmetics/Weapons/FORT_Melee/Pickaxe_VowelSleep/FX/NS_Pickaxe_VoweSleep_Trail
dependencies {
 usedDataInterfaces [&NiagaraDataInterfaceSocketReader, &NiagaraDataInterfaceCurve]
 usedDynamicInputs [
  &MergeHelper_FloatSwitch
  &FNTimeOfDayColor
  &ClampFloat
  &RemapRange
  &Add_Float
  &Subtract_Float
  &FloatFromCurve
 ]
 usedModules [
  &self:SystemSpawnScript.SystemScriptSource.SystemScriptGraph.NiagaraNodeAssignment_1.SetVariables_#1
  &self:SystemSpawnScript.SystemScriptSource.SystemScriptGraph.NiagaraNodeAssignment_3.SetVariables_#2
  &SystemState
  &self:SystemSpawnScript.SystemScriptSource.SystemScriptGraph.NiagaraNodeAssignment_9.SetVariables_#3
  &self:SystemSpawnScript.SystemScriptSource.SystemScriptGraph.NiagaraNodeAssignment_6.SetVariables_#4
  &self:SystemSpawnScript.SystemScriptSource.SystemScriptGraph.NiagaraNodeAssignment_11.SetVariables_#5
 ]
}
emitters {TrailBase002 {
  emitterUpdateScript [
   &EmitterState()
   &FN_SpawnAnimTrail(Number of Anim Trails=2)
  ]
  particleSpawnScript [
   &FN_InitializeAnimTrailBoneIndecies(AnimTrail Selection=1, Second Bone or Socket Index=2)
   &FN_InitializeAnimTrailBoneIndecies() {module=FN_InitializeAnimTrailBoneIndecies001}
   &FN_InitializeAnimTrail(
    Color=&MultiplyLinearColor(
     Linear Color=&MaskLinearColorBySpawnGroup(Masked Linear Color=<rgba 0.1878 0.1972 0.2348 0.4261>, Passthrough Linear Color=<rgba 0.3565 0.2928 0.155 0.5304>)
     Scale Factor=&MakeVector4(X=1, Y=1, Z=1, W=$System.FadeIn)
    )
    Lifetime=&MaskFloatBySpawnGroup(Masked Float=0.15, Passthrough Float=0.1)
   )
  ]
  particleUpdateScript [
   &FN_UpdateAnimTrail() {module=FN_UpdateAnimTrail001}
   &ScaleColor(
    Scale Mode=1
    Scale RGBA=&Multiply_Vector4(
     A=&MakeVector4_FromLinearColor(LinearColor=$System.TODColor)
     B=&MakeVector4FromVectorAndFloat("Vector (RGB)"=<1 1 1>, "Float (Alpha)"=$System.FadeOut)
    )
    ScaleRGBA=true
   ) {module=ScaleColor002}
  ]
  rendererClasses [&NiagaraRibbonRendererProperties]
  renderers [{rendererClass=&NiagaraRibbonRendererProperties}]
 }}
summary {
 systemName=NS_Pickaxe_VoweSleep_Trail
 emitters [{emitterName=TrailBase002, rendererClasses [&NiagaraRibbonRendererProperties]}]
 userVariables [
  User.bIsFrontEnd {type=NiagaraBool}
  User.bIsFrontEndPreview {type=NiagaraBool}
  User.RibbonIDOffset {type=NiagaraInt32}
  User.SwingSectionIndex {type=NiagaraInt32}
 ]
}
